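# Debug tool widget in QuestUI that represents one schedule bundle entry, with info fields for bundle definition, display data, unlock type and unlock value.
# SetData stores the entry in ActiveEntry with a field-notify broadcast; the event graph is empty.
# Helper functions convert data for display: asset names with a not-found fallback, unlock type text, a pip color by entry type, and dimmed content opacity without a bundle view.
# The layout also holds select and browse buttons.
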WidgetBlueprint /QuestUI/DebugTool/ScheduleWidgets/QuestDebugToolScheduleEntry
graphs [
 EventGraph {links [], nodes []}
 Get Asset Name {nodes [
   n1 = Get Asset Name!() -> (then, Asset:Primary Data Asset Object Reference) {is=title K2Node_FunctionEntry}
   > n2 = Is Valid(InputObject=@n1)
   Return Node!(execute=@n2.Is Not Valid, AssetName:Text="LOCTEXT(\"[Not found]\")")
   > Return Node(AssetName="To Text (String)"(InString=GetObjectName(Object=@n1)))
  ]}
 Get Content Opacity {nodes [
   n1 = Get Content Opacity!() -> (then, Entry:"Schedule Bundle Entry Debug Data Structure (by ref)") {is=title K2Node_FunctionEntry}
   > Return Node!(
    execute
    Opacity:double=SelectFloat(
     A=1
     B=0.5
     bPickA=Break Schedule Bundle Entry Debug Data!(ScheduleBundleEntryDebugData:"Schedule Bundle Entry Debug Data Structure (by ref)"=@n1) -> (BundleListItemId:String, bHasBundleView:bool=~) {is=title K2Node_BreakStruct}.bHasBundleView
    )
   )
  ]}
 Get Entry Unlock Type {nodes [
   n1 = Get Entry Unlock Type!() -> (then, Entry:"Schedule Bundle Entry Debug Data Structure (by ref)") {is=title K2Node_FunctionEntry}
   > Return Node!(
    execute
    UnlockType:Text=To Text!(Value:EChallengeScheduleUnlockType Enum=Break Schedule Bundle Entry Debug Data!(ScheduleBundleEntryDebugData=@n1) -> (BundleListItemId, UnlockType:EChallengeScheduleUnlockType Enum).UnlockType) -> (ReturnValue:Text) {is=title K2Node_GenericToText}
   )
  ]}
 Get Pip Color {nodes [
   n1 = Get Pip Color!() -> (then, Entry:"Schedule Bundle Entry Debug Data Structure (by ref)") {is=title K2Node_FunctionEntry}
   n2 = Break Schedule Bundle Entry Debug Data!(ScheduleBundleEntryDebugData=@n1) -> (BundleListItemId, bHasBundleView:bool=~)
   > Return Node!(
    execute
    Color:Linear Color Structure=Get Debug List Entry Type Color!(EntryType:EQuestPanelDebugListEntryType Enum=Select(Option 0:EQuestPanelDebugListEntryType Enum=BundleAsset, Option 1:EQuestPanelDebugListEntryType Enum=BundleView, Index=@n2.bHasBundleView) -> (ReturnValue:EQuestPanelDebugListEntryType Enum)) -> (Color:Linear Color Structure) {is=title K2Node_CallFunction}
   )
  ]}
 SetData {nodes [
   n1 = SetData!() -> (then, Entry:"Schedule Bundle Entry Debug Data Structure (by ref)") {is=title K2Node_FunctionEntry}
   > Set with Broadcast ActiveEntry!(execute, ActiveEntry:Schedule Bundle Entry Debug Data Structure=@n1) -> (then, Output_Get:Schedule Bundle Entry Debug Data Structure)
  ]}
]
variables [ActiveEntry {
  type=ScheduleBundleEntryDebugData
  default {bHasBundleView=false, bundleDefinition=null, bundleDisplayData=null, bundleListItemId="", unlockType=Manually, unlockValue=0}
 }]
widgets {
 info {inheritedWidgetCount=0, parentClass=/Script/QuestUI.QuestPanelDebugToolWidgetBase, rootWidgetClass=&HorizontalBox, widgetCount=12}
 widgets [
  HorizontalBox_1 {widgetClassPath=&HorizontalBox, widget=HorizontalBox_1}
  Button_SelectEntry {
   widgetClassPath=/QuestUI/DebugTool/CommonWidgets/QuestDebugToolSelectEntryButton
   bIsVariable=true
   parent=HorizontalBox_1
   widget=Button_SelectEntry
   slot {class=HorizontalBoxSlot, properties {verticalAlignment=VAlign_Center, padding=<ltrb 4 4 4 4>}}
  }
  Pip {
   widgetClassPath=&Image
   bIsVariable=true
   parent=HorizontalBox_1
   widget=Pip
   properties {brush {
     bIsDynamicallyLoaded=false
     drawAs=Image
     imageType=NoImage
     mirroring=NoMirror
     resourceName=None
     resourceObject=null
     tiling=NoTile
     imageSize=<8 8>
     margin=<ltrb 0 0 0 0>
     outlineSettings {
      bUseBrushTransparency=false
      roundingType=HalfHeightRadius
      width=0
      color {colorUseRule=UseColor_Specified, specifiedColor=<rgba 0 0 0 0>}
      cornerRadii=<0 0 0 0>
     }
     tintColor {colorUseRule=UseColor_Specified, specifiedColor=<rgba 1 1 1 1>}
     uVRegion {bIsValid=false, max=<0 0>, min=<0 0>}
    }}
   slot {class=HorizontalBoxSlot, properties {padding=<ltrb 0 6 0 6>}}
  }
  Content {
   widgetClassPath=&GridPanel
   parent=HorizontalBox_1
   widget=Content
   properties {bOverride_Cursor=true, columnFill [4,1]}
   slot {class=HorizontalBoxSlot, properties {size {sizeRule=Fill, value=1}}}
  }
  HorizontalBox_0 {widgetClassPath=&HorizontalBox, parent=Content, widget=HorizontalBox_0, slot {class=GridSlot}}
  Button_BrowseDefinition {
   widgetClassPath=/QuestUI/DebugTool/CommonWidgets/QuestDebugToolBrowseButton
   bIsVariable=true
   parent=HorizontalBox_0
   widget=Button_BrowseDefinition
   slot {class=HorizontalBoxSlot, properties {verticalAlignment=VAlign_Center, padding=<ltrb 4 4 4 4>}}
  }
  Widget_DefinitionField {
   widgetClassPath=/QuestUI/DebugTool/CommonWidgets/QuestDebugToolInfoField
   bIsVariable=true
   parent=HorizontalBox_0
   widget=Widget_DefinitionField
   properties {header=Bundle Definition}
   slot {class=HorizontalBoxSlot, properties {size {sizeRule=Fill, value=1}}}
  }
  HorizontalBox {widgetClassPath=&HorizontalBox, parent=Content, widget=HorizontalBox, slot {class=GridSlot, properties {row=1}}}
  Button_BrowseDisplayData {
   widgetClassPath=&QuestDebugToolBrowseButton
   bIsVariable=true
   parent=HorizontalBox
   widget=Button_BrowseDisplayData
   slot {class=HorizontalBoxSlot, properties {verticalAlignment=VAlign_Center, padding=<ltrb 4 4 4 4>}}
  }
  Widget_DisplayDataField {
   widgetClassPath=&QuestDebugToolInfoField
   bIsVariable=true
   parent=HorizontalBox
   widget=Widget_DisplayDataField
   properties {header=Bundle Display Data}
   slot {class=HorizontalBoxSlot, properties {size {sizeRule=Fill, value=1}}}
  }
  Widget_UnlockType {
   widgetClassPath=&QuestDebugToolInfoField
   bIsVariable=true
   parent=Content
   widget=Widget_UnlockType
   properties {header=Unlock Type}
   slot {class=GridSlot, properties {column=1}}
  }
  Widget_UnlockValue {
   widgetClassPath=&QuestDebugToolInfoField
   bIsVariable=true
   parent=Content
   widget=Widget_UnlockValue
   properties {header=Unlock Value}
   slot {class=GridSlot, properties {column=1, row=1}}
  }
 ]
}
